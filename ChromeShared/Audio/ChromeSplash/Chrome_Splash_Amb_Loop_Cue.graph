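# Looping ambient sound cue for the Chrome Splash grenade, classed as a gadget item sound.
# A Random node picks one of three looping waves, feeding a continuous modulator that drives volume from a Velocity parameter.
# An Enveloper at the root then ramps volume up over the first second.
# It uses the grenade cook attenuation asset and overrides concurrency to a single voice with a release time when stolen.
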SoundCue /ChromeShared/Audio/ChromeSplash/Chrome_Splash_Amb_Loop_Cue
root=Enveloper_0
nodes [
 {
  id=Enveloper_0
  children [ModulatorContinuous_0]
  properties {pitchCurve {editorCurveData {keys [{time=0}]}}, volumeCurve {editorCurveData {keys [<at 0 0.3>,<at 1 1>]}}}
 }
 {id=ModulatorContinuous_0, children [Random_0], properties {volumeModulationParams {default=0, parameterName=Velocity}}}
 {id=Random_0, children [WavePlayer_0,WavePlayer_1,WavePlayer_2], properties {bShouldExcludeFromBranchCulling=true, weights [1,1,1]}}
 {id=WavePlayer_0, properties {bLooping=true, soundWaveAssetPtr=&/Chrome_Splash_Grenade_Amb_Loop_01}}
 {id=WavePlayer_1, properties {bLooping=true, soundWaveAssetPtr=&/Chrome_Splash_Grenade_Amb_Loop_02}}
 {id=WavePlayer_2, properties {bLooping=true, soundWaveAssetPtr=&/Chrome_Splash_Grenade_Amb_Loop_03}}
]
properties {
 attenuationSettings=&/Chrome_Splash_Grenade_Cook_Att
 bOverrideConcurrency=true
 maxDistance=3000
 soundClassObject=&Fort_Gadgets_Items
 volumeMultiplier=0.2
 attenuationOverrides {
  bEnableOcclusion=true
  falloffDistance=600
  occlusionTraceChannel=ECC_GameTraceChannel11
  occlusionVolumeAttenuation=0.2
  attenuationShapeExtents {x=250}
 }
 concurrencyOverrides {maxCount=1, resolutionRule=StopLowestPriority, voiceStealReleaseTime=2}
}
