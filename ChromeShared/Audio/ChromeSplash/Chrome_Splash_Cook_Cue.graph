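# Sound cue for the cooking sound of the Chrome Splash grenade, routed through the Fort_Gadgets_Items sound class.
# A Random node picks one of three equally weighted cook waves, excluded from branch culling, then passes it through a Modulator.
# It uses a dedicated attenuation asset and a concurrency override allowing one instance per owner, stopping the oldest when a new one plays.
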SoundCue /ChromeShared/Audio/ChromeSplash/Chrome_Splash_Cook_Cue
root=Modulator_0
nodes [
 {id=Modulator_0, children [Random_0]}
 {id=Random_0, children [WavePlayer_1,WavePlayer_2,WavePlayer_3], properties {bShouldExcludeFromBranchCulling=true, weights [1,1,1]}}
 {id=WavePlayer_1, properties {soundWaveAssetPtr=&/Chrome_Splash_Grenade_Cook_01}}
 {id=WavePlayer_2, properties {soundWaveAssetPtr=&/Chrome_Splash_Grenade_Cook_02}}
 {id=WavePlayer_3, properties {soundWaveAssetPtr=&/Chrome_Splash_Grenade_Cook_03}}
]
properties {
 attenuationSettings=&/Chrome_Splash_Grenade_Cook_Att
 bOverrideConcurrency=true
 maxDistance=3000
 soundClassObject=&Fort_Gadgets_Items
 volumeMultiplier=1.2
 concurrencyOverrides {bLimitToOwner=true, maxCount=1, resolutionRule=StopOldest}
}
