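# Sound cue played when the Chrome Splash item is dropped.
# A mixer layers a reused Chug Bomb drop wave, pitched over a random range, with one of three randomly chosen Chrome Splash grenade equip waves shifted to a fixed higher pitch.
# It uses weapon-drop attenuation and the gadget-items sound class, with a concurrency override that keeps only a couple of instances playing by stopping the lowest priority.
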SoundCue /ChromeShared/Audio/ChromeSplash/Chrome_Splash_Drop_Cue
root=Mixer_0
nodes [
 {id=Mixer_0, children [Modulator_0,Modulator_1], properties {inputVolume [0.5,1.2]}}
 {id=Modulator_0, children [WavePlayer_0], properties {pitchMax=1.5, pitchMin=0.85, volumeMax=1, volumeMin=1}}
 {id=Modulator_1, children [Random_0], properties {pitchMax=1.3, pitchMin=1.3}}
 {id=Random_0, children [WavePlayer_1,WavePlayer_2,WavePlayer_3], properties {bShouldExcludeFromBranchCulling=true, weights [1,1,1]}}
 {id=WavePlayer_0, properties {soundWaveAssetPtr=/Game/Athena/Sounds/Items/Chug_Bomb/Chug_Bomb_Drop_01}}
 {id=WavePlayer_1, properties {soundWaveAssetPtr=&/Chrome_Splash_Grenade_Equip_01}}
 {id=WavePlayer_2, properties {soundWaveAssetPtr=&/Chrome_Splash_Grenade_Equip_02}}
 {id=WavePlayer_3, properties {soundWaveAssetPtr=&/Chrome_Splash_Grenade_Equip_03}}
]
properties {
 attenuationSettings=&Fort_Weapon_Drop_Att
 bOverrideConcurrency=true
 maxDistance=3000
 soundClassObject=&Fort_Gadgets_Items
 volumeMultiplier=0.7
 concurrencyOverrides {maxCount=2, resolutionRule=StopLowestPriority}
}
