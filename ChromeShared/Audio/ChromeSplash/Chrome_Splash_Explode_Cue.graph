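# Sound cue for the Chrome Splash grenade's close explosion, layering glass and goop sounds.
# A mixer plays both layers at once, each picking randomly from its own equally weighted set of wave variations and passing through a pitch modulator.
# An external attenuation asset is assigned, and the cue uses the large explosives sound class.
# A concurrency override caps simultaneous instances, stopping the lowest priority and applying volume scaling to overlapping plays.
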SoundCue /ChromeShared/Audio/ChromeSplash/Chrome_Splash_Explode_Cue
root=Mixer_0
nodes [
 {id=Mixer_0, children [Modulator_3,Modulator_2], properties {inputVolume [1,1]}}
 {id=Modulator_2, children [Random_0], properties {pitchMax=1.25, volumeMax=1, volumeMin=1}}
 {id=Modulator_3, children [Random_3], properties {pitchMax=1.2, volumeMax=1, volumeMin=1}}
 {
  id=Random_0
  children [WavePlayer_0, WavePlayer_1, WavePlayer_2, WavePlayer_3, WavePlayer_7, WavePlayer_15]
  properties {bShouldExcludeFromBranchCulling=true, weights [1,1,1,1,1,1]}
 }
 {
  id=Random_3
  children [WavePlayer_16, WavePlayer_17, WavePlayer_18, WavePlayer_19]
  properties {bShouldExcludeFromBranchCulling=true, weights [1,1,1,1]}
 }
 {id=WavePlayer_0, properties {soundWaveAssetPtr=&/Chrome_Splash_Grenade_Explo_Close_Goop_v2_01}}
 {id=WavePlayer_1, properties {soundWaveAssetPtr=&/Chrome_Splash_Grenade_Explo_Close_Goop_v2_02}}
 {id=WavePlayer_15, properties {soundWaveAssetPtr=&/Chrome_Splash_Grenade_Explo_Close_Goop_v2_06}}
 {id=WavePlayer_16, properties {soundWaveAssetPtr=&/Chrome_Splash_Grenade_Explo_Close_Glass_v2_04}}
 {id=WavePlayer_17, properties {soundWaveAssetPtr=&/Chrome_Splash_Grenade_Explo_Close_Glass_v2_01}}
 {id=WavePlayer_18, properties {soundWaveAssetPtr=&/Chrome_Splash_Grenade_Explo_Close_Glass_v2_02}}
 {id=WavePlayer_19, properties {soundWaveAssetPtr=&/Chrome_Splash_Grenade_Explo_Close_Glass_v2_03}}
 {id=WavePlayer_2, properties {soundWaveAssetPtr=&/Chrome_Splash_Grenade_Explo_Close_Goop_v2_03}}
 {id=WavePlayer_3, properties {soundWaveAssetPtr=&/Chrome_Splash_Grenade_Explo_Close_Goop_v2_04}}
 {id=WavePlayer_7, properties {soundWaveAssetPtr=&/Chrome_Splash_Grenade_Explo_Close_Goop_v2_05}}
]
properties {
 attenuationSettings=&/Chrome_Splash_Explode_Att
 bOverrideConcurrency=true
 maxDistance=16000
 soundClassObject=/Game/Sounds/FortSoundClassesAndModes/Fort_Explosives_Large_PSM
 volumeMultiplier=0.4246
 attenuationOverrides {bAttenuate=false, bSpatialize=false, falloffDistance=2400, attenuationShapeExtents {x=600}}
 concurrencyOverrides {bVolumeScaleCanRelease=true, maxCount=3, resolutionRule=StopLowestPriority, volumeScale=0.5, volumeScaleAttackTime=0.1}
}
