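# Sound cue for a close grenade explosion in water, part of the Chrome Splash audio set and routed to the large explosives sound class.
# A random node picks one of four equally weighted wave variations and feeds a modulator that raises the volume minimum to full.
# A shared explosion attenuation asset handles falloff, while a concurrency override caps instances at three, stops the lowest priority and scales down the volume of overlapping plays.
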SoundCue /ChromeShared/Audio/ChromeSplash/Chrome_Splash_Explode_Water_Cue
root=Modulator_2
nodes [
 {id=Modulator_2, children [Random_2], properties {volumeMin=1}}
 {id=Random_2, children [WavePlayer_7, WavePlayer_9, WavePlayer_10, WavePlayer_11], properties {weights [1,1,1,1]}}
 {id=WavePlayer_10, properties {soundWaveAssetPtr=&/Chrome_Splash_Grenade_Explo_Close_Water_v2_03}}
 {id=WavePlayer_11, properties {soundWaveAssetPtr=&/Chrome_Splash_Grenade_Explo_Close_Water_v2_01}}
 {id=WavePlayer_7, properties {soundWaveAssetPtr=&/Chrome_Splash_Grenade_Explo_Close_Water_v2_04}}
 {id=WavePlayer_9, properties {soundWaveAssetPtr=&/Chrome_Splash_Grenade_Explo_Close_Water_v2_02}}
]
properties {
 attenuationSettings=&/Chrome_Splash_Explode_Att
 bOverrideConcurrency=true
 maxDistance=16000
 soundClassObject=/Game/Sounds/FortSoundClassesAndModes/Fort_Explosives_Large_PSM
 volumeMultiplier=0.4246
 attenuationOverrides {bAttenuate=false, bSpatialize=false, falloffDistance=2400, attenuationShapeExtents {x=600}}
 concurrencyOverrides {bVolumeScaleCanRelease=true, maxCount=3, resolutionRule=StopLowestPriority, volumeScale=0.5, volumeScaleAttackTime=0.1}
}
