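# Sound cue for the Chrome Splash landing, playing one of two grenade drop sounds.
# A Random node picks between the two equally weighted waves and feeds a Modulator that slightly varies pitch.
# The cue overrides attenuation with a short falloff distance and caps concurrency at a few instances, refusing new plays at the limit.
# It routes through the Fort_Loot_Collect sound class.
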SoundCue /ChromeShared/Audio/ChromeSplash/Chrome_Splash_Landed_Cue
root=Modulator_1
nodes [
 {id=Modulator_1, children [Random_1], properties {pitchMax=1.1, pitchMin=0.9}}
 {id=Random_1, children [WavePlayer_0,WavePlayer_4], properties {bShouldExcludeFromBranchCulling=true, weights [1,1]}}
 {id=WavePlayer_0, properties {soundWaveAssetPtr=&/Chrome_Splash_Grenade_Drop_02}}
 {id=WavePlayer_4, properties {soundWaveAssetPtr=&/Chrome_Splash_Grenade_Drop_01}}
]
properties {
 attenuationSettings=&Foley_Attenuation_Basic_Att
 bOverrideAttenuation=true
 bOverrideConcurrency=true
 maxDistance=1000
 soundClassObject=&Fort_Loot_Collect
 volumeMultiplier=1
 attenuationOverrides {falloffDistance=1000, attenuationShapeExtents {x=0}}
 concurrencyOverrides {maxCount=3, resolutionRule=PreventNew}
}
