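# Sound cue for picking up the Chrome Splash grenade item, built from its equip sound waves.
# An equally weighted Random node picks one of three equip variations, excluded from branch culling, and feeds a Modulator.
# It uses weapon pickup attenuation and the gadgets/items sound class.
# A concurrency override allows two simultaneous plays, stopping the lowest priority one beyond that.
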SoundCue /ChromeShared/Audio/ChromeSplash/Chrome_Splash_Pickup_Cue
root=Modulator_0
nodes [
 {id=Modulator_0, children [Random_0]}
 {id=Random_0, children [WavePlayer_1,WavePlayer_2,WavePlayer_3], properties {bShouldExcludeFromBranchCulling=true, weights [1,1,1]}}
 {id=WavePlayer_1, properties {soundWaveAssetPtr=&/Chrome_Splash_Grenade_Equip_01}}
 {id=WavePlayer_2, properties {soundWaveAssetPtr=&/Chrome_Splash_Grenade_Equip_02}}
 {id=WavePlayer_3, properties {soundWaveAssetPtr=&/Chrome_Splash_Grenade_Equip_03}}
]
properties {
 attenuationSettings=&Fort_Weapon_Pickup_Att
 bOverrideConcurrency=true
 maxDistance=3500
 soundClassObject=&Fort_Gadgets_Items
 volumeMultiplier=1
 concurrencyOverrides {maxCount=2, resolutionRule=StopLowestPriority}
}
