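# Sound cue for the Chrome Splash grenade projectile in flight, mixing an in-air loop with a whoosh layer.
# The loop fades in through an Enveloper, then passes through slight pitch modulation and a smoothed Doppler node for movement-based pitch shift.
# The whoosh layer randomly picks one of three waves and modulates it.
# The cue uses grenade projectile attenuation and overrides concurrency to cap simultaneous instances, stopping the lowest priority.
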SoundCue /ChromeShared/Audio/ChromeSplash/Chrome_Splash_Prj_InAir_Loop_Cue
root=Mixer_0
nodes [
 {id=Doppler_0, children [Modulator_0], properties {bUseSmoothing=true, dopplerIntensity=1.5, smoothingInterpSpeed=10}}
 {
  id=Enveloper_0
  children [WavePlayer_0]
  properties {pitchCurve {editorCurveData {keys [{time=0}]}}, volumeCurve {editorCurveData {keys [<at 0 0 RCIM_Cubic>,<at 1 1>]}}}
 }
 {id=Mixer_0, children [Doppler_0,Modulator_1], properties {inputVolume [0.6,0.9]}}
 {id=Modulator_0, children [Enveloper_0], properties {pitchMax=1.1, pitchMin=0.9, volumeMax=1, volumeMin=1}}
 {id=Modulator_1, children [Random_0], properties {pitchMax=1.2}}
 {id=Random_0, children [WavePlayer_1,WavePlayer_2,WavePlayer_3], properties {bShouldExcludeFromBranchCulling=true, weights [1,1,1]}}
 {id=WavePlayer_0, properties {bLooping=true, soundWaveAssetPtr=&/Chrome_Splash_Grenade_Prj_InAir_Loop}}
 {id=WavePlayer_1, properties {soundWaveAssetPtr=&/Chrome_Splash_Grenade_Prj_InAir_Whsh_01}}
 {id=WavePlayer_2, properties {soundWaveAssetPtr=&/Chrome_Splash_Grenade_Prj_InAir_Whsh_02}}
 {id=WavePlayer_3, properties {soundWaveAssetPtr=&/Chrome_Splash_Grenade_Prj_InAir_Whsh_03}}
]
properties {
 attenuationSettings=&/Chrome_Splash_Grenade_Prj_Att
 bOverrideConcurrency=true
 maxDistance=7500
 pitchMultiplier=1.65
 soundClassObject=&Fort_WeaponRanged_Fire_3P_PSM
 volumeMultiplier=1.5
 concurrencyOverrides {maxCount=3, resolutionRule=StopLowestPriority}
}
