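# Sound cue for the close in-air whip of a Chrome Splash grenade projectile, classed as a ranged weapon projectile effect.
# A Random node picks one of four equally weighted whip waves, and a Modulator varies its pitch at steady volume.
# A Quality Level root feeds that same chain to its first three inputs and leaves the fourth empty.
# A separate attenuation asset shapes falloff, and a concurrency override caps instances, stopping the lowest priority.
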SoundCue /ChromeShared/Audio/ChromeSplash/Chrome_Splash_Prj_Whip_Close_Cue
root=QualityLevel_0
nodes [
 {id=Modulator_0, children [Random_1], properties {pitchMax=1.2, volumeMax=1, volumeMin=1}}
 {id=QualityLevel_0, children [Modulator_0,Modulator_0,Modulator_0,null]}
 {id=Random_1, children [WavePlayer_3,WavePlayer_4,WavePlayer_5,WavePlayer_6], properties {weights [1,1,1,1]}}
 {id=WavePlayer_3, properties {soundWaveAssetPtr=&/Chrome_Splash_Grenade_Prj_InAir_Whip_01}}
 {id=WavePlayer_4, properties {soundWaveAssetPtr=&/Chrome_Splash_Grenade_Prj_InAir_Whip_02}}
 {id=WavePlayer_5, properties {soundWaveAssetPtr=&/Chrome_Splash_Grenade_Prj_InAir_Whip_03}}
 {id=WavePlayer_6, properties {soundWaveAssetPtr=&/Chrome_Splash_Grenade_Prj_InAir_Whip_04}}
]
properties {
 attenuationSettings=&/Chrome_Splash_Prj_Whip_Att
 bOverrideConcurrency=true
 maxDistance=8500
 soundClassObject=&Fort_WeaponRanged_Projectile_Effects
 volumeMultiplier=1
 attenuationOverrides {
  bAttenuate=false
  bEnableReverbSend=false
  bSpatialize=false
  bUseComplexCollisionForOcclusion=true
  falloffDistance=1000
  occlusionVolumeAttenuation=0
  stereoSpread=0
  attenuationShapeExtents {x=512}
 }
 concurrencyOverrides {maxCount=4, resolutionRule=StopLowestPriority}
}
