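# Sound cue for throwing the Chrome Splash grenade, routed to the gadgets and items sound class.
# A random node picks one of three equally weighted throw waves, then a modulator varies pitch and volume.
# It uses a dedicated attenuation asset with virtualization disabled.
# A concurrency override limits it to two instances, stopping the lowest priority and scaling down volume as instances overlap.
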SoundCue /ChromeShared/Audio/ChromeSplash/Chrome_Splash_Throw_Cue
root=Modulator_2
nodes [
 {id=Modulator_2, children [Random_0], properties {pitchMax=1.2, volumeMax=1}}
 {id=Random_0, children [WavePlayer_1,WavePlayer_2,WavePlayer_3], properties {bShouldExcludeFromBranchCulling=true, weights [1,1,1]}}
 {id=WavePlayer_1, properties {soundWaveAssetPtr=&/Chrome_Splash_Grenade_Throw_01}}
 {id=WavePlayer_2, properties {soundWaveAssetPtr=&/Chrome_Splash_Grenade_Throw_02}}
 {id=WavePlayer_3, properties {soundWaveAssetPtr=&/Chrome_Splash_Grenade_Throw_03}}
]
properties {
 attenuationSettings=&/Chrome_Splash_Grenade_Throw_Att
 bOverrideConcurrency=true
 maxDistance=6000
 soundClassObject=&Fort_Gadgets_Items
 virtualizationMode=Disabled
 concurrencyOverrides {bVolumeScaleCanRelease=true, maxCount=2, resolutionRule=StopLowestPriority, volumeScale=0.5, volumeScaleAttackTime=0.1}
}
